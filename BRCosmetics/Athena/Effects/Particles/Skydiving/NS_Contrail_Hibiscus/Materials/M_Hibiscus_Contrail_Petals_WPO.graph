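# Translucent, unlit, two-sided mesh-particle material for Hibiscus skydiving contrail petals and leaves.
# A Niagara dynamic WPOAngle bends vertices around an instance-local axis perpendicular to their radial direction, gated by the UseWPO? scalar.
# A synced distance fade adds an offset shrinking particles toward their centers, alongside a near-camera fade.
# UseFlowerTexture? blends leaf and petal textures, tinted by particle color, into emissive adjustments.
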Material /BRCosmetics/Athena/Effects/Particles/Skydiving/NS_Contrail_Hibiscus/Materials/M_Hibiscus_Contrail_Petals_WPO
graph {
 properties {
  bUsedWithNiagaraMeshParticles=true
  blendMode=BLEND_Translucent
  pixelDepthOffsetMode=PDOM_Legacy
  shadingModel=MSM_Unlit
  twoSided=true
  usedShadingModels=Unlit
 }
 nodes [
  n1 = AppendVector(
   A=ComponentMask(None=MaterialFunctionCall() {materialFunction=&LocalPosition}."Instance Local Position (Excluding Offsets)") {g=true, r=true}
   B=Constant()
  )
  n2 = RotateAboutAxis(
   NormalizedRotationAxis=CrossProduct(A=Constant3Vector() {constant=<rgba 0 0 1 0>}, B=Normalize(VectorInput=@n1))
   RotationAngle=DynamicParameter() {defaultValue=<rgba 0 1 1 1>, paramNames [WPOAngle,Param2,Param3,Param4]}.WPOAngle
   PivotPoint=Constant3Vector()
   Position=MaterialFunctionCall() {materialFunction=&LocalPosition}."Instance Local Position (Excluding Offsets)"
  ) {desc="!! RotateAboutAxis outputs DELTAS not absolute Positions"}
  n3 = ParticlePositionWS()
  n4 = WorldPosition() {worldPositionShaderOffset=WPT_ExcludeAllShaderOffsets}
  n5 = ScalarParameter() {defaultValue=150, parameterName=FadeStartDistance}
  n6 = ParticleRadius()
  n7 = ScalarParameter() {defaultValue=50, parameterName=FadeEndDistance}
  n8 = LinearInterpolate(
   A=ScalarParameter() {defaultValue=0.1}
   Alpha=MaterialFunctionCall(Position=ParticlePositionWS(), Fade_Start_Distance=Add(A=@n5, B=@n6), Fade_End_Distance=Add(A=@n7, B=@n6)) {materialFunction=&MF_Distance_SyncedFade_FX}.Result
  ) {constA=0.2}
  Add(
   A=LinearInterpolate(B=Transform(None=@n2) {transformSourceType=TRANSFORMSOURCE_Instance}, Alpha=ScalarParameter() {parameterName=UseWPO?}) {constB=0}
   B=Multiply(A=Subtract(A=@n3, B=@n4.XYZ), B=OneMinus(None=@n8))
  )
  MaterialFunctionCall(Distance=ScalarParameter() {defaultValue=99, parameterName=Param_1}) {materialFunction=&Near_Camera_Fade}
  n9 = TextureSample() {texture=/BRCosmetics/Athena/Effects/Particles/Skydiving/NS_Contrail_Hibiscus/Textures/T_Hibiscus_Leaf}
  n10 = ParticleColor()
  n11 = TextureSample() {texture=/BRCosmetics/Athena/Effects/Particles/Skydiving/NS_Contrail_Hibiscus/Textures/T_Hibiscus_FlowerPetal}
  n12 = ParticleColor()
  MaterialFunctionCall(Emissive=LinearInterpolate(A=Multiply(A=@n9.RGB, B=@n10.RGB), B=Multiply(A=@n11.RGB, B=@n12.RGB), Alpha=ScalarParameter() {parameterName=UseFlowerTexture?})) {materialFunction=&MF_FX_EmissiveAdjustments}
 ]
}
