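# Niagara skydiving contrail for the Hibiscus cosmetic, trailing vines, leaves and flowers from the player.
# Flower and leaf meshes follow Vine_L and Vine_R through particle attribute readers and generate death events, alongside detached-petal emitters.
# Player skeletal mesh sockets anchor hand wraps, glows and leaf bursts, while GPU emitters swirl petals and leaves around the player.
# User.ParaGlideLeanAlpha scales several spawn rates and velocities, with front-end overrides on some.
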NiagaraSystem /BRCosmetics/Athena/Effects/Particles/Skydiving/NS_Contrail_Hibiscus/NS_Hibiscus_Vines_Trail
dependencies {
 usedDataInterfaces [
  &NiagaraDataInterfaceCurve
  &NiagaraDataInterfaceArrayInt32
  &NiagaraDataInterfaceVector2DCurve
  &NiagaraDataInterfaceParticleRead
  &NiagaraDataInterfaceArrayFloat
  &NiagaraDataInterfaceVectorCurve
  &NiagaraDataInterfaceArrayColor
 ]
 usedDynamicInputs [
  &SelectLinearColorFromArray
  &NormalizeVector
  &MakeVector
  &MakeFloatFromVector
  &Multiply_Float
  &Abs_Float
 ]
 usedModules [
  &SystemState
  &self:SystemSpawnScript.SystemScriptSource.SystemScriptGraph.NiagaraNodeAssignment_0.SetVariables_#1
  &self:SystemSpawnScript.SystemScriptSource.SystemScriptGraph.NiagaraNodeAssignment_1.SetVariables_#2
  &InterpolateOverTime
  &self:SystemSpawnScript.SystemScriptSource.SystemScriptGraph.NiagaraNodeAssignment_2.SetVariables_#3
 ]
}
emitters {
 FlowerPetals_AroundPlayer {
  simTarget=GPUComputeSim
  emitterUpdateScript [
   &EmitterState()
   &SpawnRate(
    SpawnRate=&FN_FrontEndFloatModifier(Gameplay=&Lerp_Float(A=3, B=5, Alpha=$User.ParaGlideLeanAlpha), Front End=20)
    Use Spawn Probability=false
   )
  ]
  particleSpawnScript [
   &InitializeParticle(
    Color=&RandomRangeLinearColor(Minimum=<rgba 1 0.2042 0.8167 1>, Maximum=<rgba 1 0.2232 0.1403 1>)
    Lifetime Min=0.5
    Lifetime Mode=::NewEnumerator1
    Mesh Scale Mode=::NewEnumerator1
    Mesh Uniform Scale Max=0.8
    Mesh Uniform Scale Min=0.4
    Sprite Rotation Mode=::NewEnumerator2
    Sprite Size Max=<18 17>
    Sprite Size Min=<5 6>
    Sprite Size Mode=::NewEnumerator2
    UsePositionOffset=false
    Write Material Random=true
   ) {module=InitializeParticle001}
   &ShapeLocation(
    MaskSpawnGroup=false
    Offset=<0 0 -75>
    Offset Mode=::NewEnumerator0
    Sphere Distribution=::NewEnumerator1
    Sphere Radius=&RandomRangeFloat(Minimum=75, Maximum=150)
    U Position=&RandomRangeFloat(Minimum=0, Maximum=1)
    UseOwnerScale=true
    V Position=&RandomRangeFloat(Minimum=0, Maximum=1)
   )
   &InheritVelocity(Inherited Velocity Amount Scale=<0 0 -0.12>, Speed Limit=false, UseSpeedThreshold=false)
   &AddVelocity(Velocity=&FN_FrontEndVector3Modifier(
     Gameplay=&Multiply_Vector(A=&VectorFromFloat(Value=$User.ParaGlideLeanAlpha), B=<0 0 150>)
     Front End=/Niagara/DynamicInputs/Lerp/Lerp_Position(A:NiagaraPosition=<0 0 50>, B:NiagaraPosition=<1 -35 300>, Alpha:NiagaraFloat=$User.ParaGlideLeanAlpha)
    ))
   &VortexForce(UseOriginOffset=false)
   &ApplyInitialForces()
   &InitialMeshOrientation(Use Rotation Vector=false)
  ]
  particleUpdateScript [
   &ParticleState()
   &ScaleMeshSize(Scale Factor=&VectorFromFloat(Value=&FN_FrontEndFloatModifier(
      Gameplay=&FloatFromCurve(FloatCurve=DataInterface() {propertyValues {Curve {keys [
           {arriveTangent=4.1785, interpMode=RCIM_Cubic, leaveTangent=4.1785, tangentMode=RCTM_User, time=0, value=0}
           {arriveTangent=1.1179, interpMode=RCIM_Cubic, leaveTangent=1.1179, tangentMode=RCTM_User, time=0.2042, value=1}
           {arriveTangent=0.0224, interpMode=RCIM_Cubic, leaveTangent=0.0224, tangentMode=RCTM_User, time=1, value=1}
          ]}}})
      Front End=1
     )))
   &ScaleMeshSize(Scale Factor=&VectorFromFloat(Value=&FloatFromCurve(FloatCurve=DataInterface() {propertyValues {Curve {keys [
          {tangentMode=RCTM_User, time=0, value=1}
          {arriveTangent=-0.0405, interpMode=RCIM_Cubic, leaveTangent=-0.0405, tangentMode=RCTM_User, time=0.7009, value=1}
          {arriveTangent=-8.0089, interpMode=RCIM_Cubic, leaveTangent=-8.0089, tangentMode=RCTM_User, time=1, value=0}
         ]}}}))) {module=ScaleMeshSize001}
   &UpdateMeshOrientation()
   &CurlNoiseForce(Noise Frequency=1000, Noise Strength=200, Pan Noise=false)
   &SolveForcesAndVelocity(Clamp Velocity=false, Limit Acceleration=false)
   &ScaleMeshSizeBySpeed(Max Scale Factor=<1.2 1.2 1.2>, Min Scale Factor:Vector3f=<0.5 0.5 0.5>)
  ]
  rendererClasses [&NiagaraMeshRendererProperties]
  renderers [{rendererClass=&NiagaraMeshRendererProperties, rendererIndex=0}]
 }
 Flower_Leafs_L {
  emitterUpdateScript [
   &EmitterState(Life Cycle Mode=::NewEnumerator1, Loop Duration=1, Recalculate Duration Each Loop=true, UseLoopDelay=false)
   &SpawnRate(
    Spawn Probability=&RandomRangeFloat(Minimum=0.5, Maximum=1)
    SpawnRate=&Lerp_Float(A=10, B=15, Alpha=$User.ParaGlideLeanAlpha)
    Use Spawn Probability=true
   )
  ]
  particleSpawnScript [
   &PartitionParticles(Number Of Partitions=3)
   /SetVariables(Particles.VisibilityTag=$Particles.PartitionParticles.Partition) {isSetParameters=true, module=SetVariables_#4}
   &InitializeParticle(
    Color Maximum=<rgba 0.5175 0.3203 0 1>
    Color Minimum=<rgba 0.361 0.3503 0.0138 1>
    Color Mode=::NewEnumerator2
    Lifetime=2
    Mesh Scale=&Lerp_Vector(
     A=&RandomRangeVector(Minimum=<4 2 10>, Maximum=<6 5 1>)
     B=&RandomRangeVector(Minimum=<0.5 0.5 1>, Maximum=<1 1 1>)
     Alpha=&MakeFloatFromInt(Int=$Particles.PartitionParticles.Partition)
    )
    Mesh Scale Mode=::NewEnumerator4
    UsePositionOffset=false
    Write Material Random=true
   )
   /SetVariables(
    Particles.Color_Flower:LinearColor=&RandomRangeLinearColor(Minimum=<rgba 1 0.2042 0.8167 1>, Maximum=<rgba 1 0.2232 0.1403 1>)
    Particles.Color_LeafTest:LinearColor=&RandomRangeLinearColor(Minimum=<rgba 0.1697 0.1647 0.0065 1>, Maximum=<rgba 0.4697 0.4277 0.0179 1>)
   ) {isSetParameters=true, module=SetVariables_#5}
  ]
  particleUpdateScript [
   &ParticleState()
   &self:ReadVinePosition(
    InputFLoat:NiagaraFloat=&FloatFromCurve(
     FloatCurve=DataInterface() {propertyValues {Curve {keys [{tangentMode=RCTM_User, time=0, value=0}, {tangentMode=RCTM_User, time=1, value=1}]}}}
     Scale Curve=4
    )
    New Particle Attribute Reader=DataInterface() {propertyValues {EmitterBinding {bindingMode=Other, emitterName=Vine_L}}}
   ) {module=ReadVinePosition}
   /SetVariables(Particles.Position=&AddVectorToPosition(
     Position=$Particles.Position
     Vector=&Multiply_VectorByFloat(
      Vector=<0 0 1>
      Float=&FloatFromCurve(
       FloatCurve=DataInterface() {propertyValues {Curve {keys [
           {arriveTangent=2.4285, interpMode=RCIM_Cubic, leaveTangent=2.4285, tangentMode=RCTM_User, time=0.75, value=0}
           {arriveTangent=5.7402, interpMode=RCIM_Cubic, leaveTangent=5.7402, tangentMode=RCTM_User, time=1, value=1}
          ]}}}
       Scale Curve=&SelectFloatFromArray(Direct Array Index=$Particles.PartitionParticles.Partition, Float Selection Array=DataInterface() {propertyValues {FloatData [15,60,60]}})
      )
     )
    )) {isSetParameters=true, module=SetVariables_#6}
   &UpdateMeshOrientation(
    Facing Direction=&Subtract_Position(A=$Particles.Position, B=$Particles.Previous.Position)
    Facing Interpolation=true
    Facing Interpolation Rate=3
    Orientation Method=::NewEnumerator1
    Override Quat=false
    Stabilization Mode:ENiagara_OrientVectorCount=::NewEnumerator0
    TransformFacing=false
   )
   &UpdateMeshOrientation() {module=UpdateMeshOrientation001}
   &ScaleMeshSize(Scale Factor=&Lerp_Vector(
     A=<1 1 1>
     B=&VectorFromCurve(
      VectorCurve=DataInterface() {propertyValues {XCurve {keys [<at 0 3>,<at 1 1>]}, YCurve {keys [<at 0 3>,<at 1 1>]}, ZCurve {keys [<at 0 0.7>,<at 0.6 0.7>,<at 0.7 1>,<at 1 1>]}}}
      Scale Curve=<0.85 0.85 0.85>
     )
     Alpha=&SelectFloatFromArray(Direct Array Index=$Particles.VisibilityTag, Float Selection Array=DataInterface() {propertyValues {FloatData [0,1,0]}})
    )) {module=ScaleMeshSize003}
   &ScaleMeshSize(Scale Factor=&VectorFromFloat(Value=&FloatFromCurve(FloatCurve=DataInterface() {propertyValues {Curve {keys [
          {arriveTangent=1.1315, interpMode=RCIM_Cubic, leaveTangent=1.1315, tangentMode=RCTM_User, time=0, value=0}
          {interpMode=RCIM_Cubic, tangentMode=RCTM_User, time=1, value=1}
         ]}}})))
   &ScaleMeshSize(Scale Factor=&Lerp_Vector(
     A=&VectorFromFloat(Value=&FloatFromCurve(FloatCurve=DataInterface() {propertyValues {Curve {keys [
           {tangentMode=RCTM_User, time=0, value=1}
           {arriveTangent=-0.0405, interpMode=RCIM_Cubic, leaveTangent=-0.0405, tangentMode=RCTM_User, time=0.9, value=1}
           {arriveTangent=-8.0089, interpMode=RCIM_Cubic, leaveTangent=-8.0089, tangentMode=RCTM_User, time=1, value=0}
          ]}}}))
     B=<1 1 1>
     Alpha=&SelectFloatFromArray(Direct Array Index=$Particles.VisibilityTag, Float Selection Array=DataInterface() {propertyValues {FloatData [0,1,0]}})
    )) {module=ScaleMeshSize002}
   &DynamicMaterialParameters(
    Index 0 Param 1=&FloatFromCurve(FloatCurve=DataInterface() {propertyValues {Curve {keys [
         {interpMode=RCIM_Cubic, tangentMode=RCTM_User, time=0, value=-0.12}
         {interpMode=RCIM_Cubic, leaveTangent=0.8559, tangentMode=RCTM_Break, time=0.6, value=-0.12}
         {arriveTangent=0.06, interpMode=RCIM_Cubic, leaveTangent=0.7317, tangentMode=RCTM_Break, time=1, value=0.04}
        ]}}})
    Module.Param0WriteEnabled=true
    Module.Param1WriteEnabled=true
    Module.Param2WriteEnabled=true
    Module.Param3WriteEnabled=true
   )
   /SetVariables(Particles.CameraOffset_Flowers=9) {isSetParameters=true, module=SetVariables_#7}
   &GenerateDeathEvent(Use Event Probability=false)
  ]
  rendererClasses [&NiagaraMeshRendererProperties]
  renderers [
   {rendererClass=&NiagaraMeshRendererProperties, rendererIndex=0}
   {rendererClass=&NiagaraMeshRendererProperties, rendererIndex=1}
   {rendererClass=&NiagaraMeshRendererProperties, rendererIndex=2}
   {rendererClass=&NiagaraMeshRendererProperties, rendererIndex=3}
  ]
 }
 Flower_Leafs_R {
  emitterUpdateScript [
   &EmitterState(Life Cycle Mode=::NewEnumerator1, Loop Duration=1, Recalculate Duration Each Loop=true, UseLoopDelay=false)
   &SpawnRate(
    Spawn Probability=&RandomRangeFloat(Minimum=0.5, Maximum=1)
    SpawnRate=&Lerp_Float(A=10, B=15, Alpha=$User.ParaGlideLeanAlpha)
    Use Spawn Probability=true
   )
  ]
  particleSpawnScript [
   &PartitionParticles(Number Of Partitions=3)
   /SetVariables(Particles.VisibilityTag=$Particles.PartitionParticles.Partition) {isSetParameters=true, module=SetVariables_#4}
   &InitializeParticle(
    Color Maximum=<rgba 0.5175 0.3203 0 1>
    Color Minimum=<rgba 0.361 0.3503 0.0138 1>
    Color Mode=::NewEnumerator2
    Lifetime=2
    Mesh Scale=&Lerp_Vector(
     A=&RandomRangeVector(Minimum=<4 2 10>, Maximum=<6 5 1>)
     B=&RandomRangeVector(Minimum=<0.5 0.5 1>, Maximum=<1 1 1>)
     Alpha=&MakeFloatFromInt(Int=$Particles.PartitionParticles.Partition)
    )
    Mesh Scale Mode=::NewEnumerator4
    UsePositionOffset=false
    Write Material Random=true
   )
   /SetVariables(
    Particles.Color_Flower:LinearColor=&RandomRangeLinearColor(Minimum=<rgba 1 0.2042 0.8167 1>, Maximum=<rgba 1 0.2232 0.1403 1>)
    Particles.Color_LeafTest:LinearColor=&RandomRangeLinearColor(Minimum=<rgba 0.1697 0.1647 0.0065 1>, Maximum=<rgba 0.4697 0.4277 0.0179 1>)
   ) {isSetParameters=true, module=SetVariables_#5}
  ]
  particleUpdateScript [
   &ParticleState()
   &self:ReadVinePosition(
    InputFLoat:NiagaraFloat=&FloatFromCurve(
     FloatCurve=DataInterface() {propertyValues {Curve {keys [{tangentMode=RCTM_User, time=0, value=0}, {tangentMode=RCTM_User, time=1, value=1}]}}}
     Scale Curve=4
    )
    New Particle Attribute Reader=DataInterface() {propertyValues {EmitterBinding {bindingMode=Other, emitterName=Vine_R}}}
   ) {module=ReadVinePosition}
   /SetVariables(Particles.Position=&AddVectorToPosition(
     Position=$Particles.Position
     Vector=&Multiply_VectorByFloat(
      Vector=<0 0 1>
      Float=&FloatFromCurve(
       FloatCurve=DataInterface() {propertyValues {Curve {keys [
           {arriveTangent=2.4285, interpMode=RCIM_Cubic, leaveTangent=2.4285, tangentMode=RCTM_User, time=0.75, value=0}
           {arriveTangent=5.7402, interpMode=RCIM_Cubic, leaveTangent=5.7402, tangentMode=RCTM_User, time=1, value=1}
          ]}}}
       Scale Curve=&SelectFloatFromArray(Direct Array Index=$Particles.PartitionParticles.Partition, Float Selection Array=DataInterface() {propertyValues {FloatData [15,60,60]}})
      )
     )
    )) {isSetParameters=true, module=SetVariables_#6}
   &UpdateMeshOrientation(
    Facing Direction=&Subtract_Position(A=$Particles.Position, B=$Particles.Previous.Position)
    Facing Interpolation=true
    Facing Interpolation Rate=3
    Orientation Method=::NewEnumerator1
    Override Quat=false
    Stabilization Mode:ENiagara_OrientVectorCount=::NewEnumerator0
    TransformFacing=false
   )
   &UpdateMeshOrientation() {module=UpdateMeshOrientation001}
   &ScaleMeshSize(Scale Factor=&Lerp_Vector(
     A=<1 1 1>
     B=&VectorFromCurve(
      VectorCurve=DataInterface() {propertyValues {XCurve {keys [<at 0 3>,<at 1 1>]}, YCurve {keys [<at 0 3>,<at 1 1>]}, ZCurve {keys [<at 0 0.7>,<at 0.6 0.7>,<at 0.7 1>,<at 1 1>]}}}
      Scale Curve=<0.85 0.85 0.85>
     )
     Alpha=&SelectFloatFromArray(Direct Array Index=$Particles.VisibilityTag, Float Selection Array=DataInterface() {propertyValues {FloatData [0,1,0]}})
    )) {module=ScaleMeshSize002}
   &ScaleMeshSize(Scale Factor=&VectorFromFloat(Value=&FloatFromCurve(FloatCurve=DataInterface() {propertyValues {Curve {keys [
          {arriveTangent=1.1315, interpMode=RCIM_Cubic, leaveTangent=1.1315, tangentMode=RCTM_User, time=0, value=0}
          {interpMode=RCIM_Cubic, tangentMode=RCTM_User, time=1, value=1}
         ]}}})))
   &ScaleMeshSize(Scale Factor=&Lerp_Vector(
     A=&VectorFromFloat(Value=&FloatFromCurve(FloatCurve=DataInterface() {propertyValues {Curve {keys [
           {tangentMode=RCTM_User, time=0, value=1}
           {arriveTangent=-0.0405, interpMode=RCIM_Cubic, leaveTangent=-0.0405, tangentMode=RCTM_User, time=0.9, value=1}
           {arriveTangent=-8.0089, interpMode=RCIM_Cubic, leaveTangent=-8.0089, tangentMode=RCTM_User, time=1, value=0}
          ]}}}))
     B=<1 1 1>
     Alpha=&SelectFloatFromArray(Direct Array Index=$Particles.VisibilityTag, Float Selection Array=DataInterface() {propertyValues {FloatData [0,1,0]}})
    )) {module=ScaleMeshSize003}
   &DynamicMaterialParameters(
    Index 0 Param 1=&FloatFromCurve(FloatCurve=DataInterface() {propertyValues {Curve {keys [
         {interpMode=RCIM_Cubic, tangentMode=RCTM_User, time=0, value=-0.12}
         {interpMode=RCIM_Cubic, leaveTangent=0.8559, tangentMode=RCTM_Break, time=0.6, value=-0.12}
         {arriveTangent=0.06, interpMode=RCIM_Cubic, leaveTangent=0.7317, tangentMode=RCTM_Break, time=1, value=0.06}
        ]}}})
    Module.Param0WriteEnabled=true
    Module.Param1WriteEnabled=true
    Module.Param2WriteEnabled=true
    Module.Param3WriteEnabled=true
   ) {enabled=false}
   &DynamicMaterialParameters(
    Index 0 Param 1=&FloatFromCurve(FloatCurve=DataInterface() {propertyValues {Curve {keys [
         {interpMode=RCIM_Cubic, tangentMode=RCTM_User, time=0, value=-0.12}
         {interpMode=RCIM_Cubic, leaveTangent=0.8559, tangentMode=RCTM_Break, time=0.6, value=-0.12}
         {arriveTangent=0.06, interpMode=RCIM_Cubic, leaveTangent=0.7317, tangentMode=RCTM_Break, time=1, value=0.04}
        ]}}})
    Module.Param0WriteEnabled=true
    Module.Param1WriteEnabled=true
    Module.Param2WriteEnabled=true
    Module.Param3WriteEnabled=true
   ) {module=DynamicMaterialParameters001}
   /SetVariables(Particles.CameraOffset_Flowers=9) {isSetParameters=true, module=SetVariables_#7}
   &GenerateDeathEvent(Use Event Probability=false)
  ]
  rendererClasses [&NiagaraMeshRendererProperties]
  renderers [
   {rendererClass=&NiagaraMeshRendererProperties, rendererIndex=0}
   {rendererClass=&NiagaraMeshRendererProperties, rendererIndex=1}
   {rendererClass=&NiagaraMeshRendererProperties, rendererIndex=2}
   {rendererClass=&NiagaraMeshRendererProperties, rendererIndex=3}
  ]
 }
 Flower_Petals_DetatchEvent_L {
  emitterUpdateScript [
   &EmitterState(Life Cycle Mode=::NewEnumerator1, Loop Duration=1, Recalculate Duration Each Loop=true, UseLoopDelay=false)
   &SpawnRate(Spawn Probability=&RandomRangeFloat(Minimum=0.5, Maximum=1), SpawnRate=10, Use Spawn Probability=true) {enabled=false}
  ]
  particleSpawnScript [
   &PartitionParticles(Number Of Partitions=3)
   /SetVariables(Particles.VisibilityTag=$Particles.PartitionParticles.Partition) {isSetParameters=true, module=SetVariables_#4}
   &InitializeParticle(
    Color Maximum=<rgba 0.5175 0.3203 0 1>
    Color Minimum=<rgba 0.361 0.3503 0.0138 1>
    Color Mode=::NewEnumerator2
    Lifetime=&RandomRangeFloat(Minimum=0.5, Maximum=1.25)
    Mesh Scale Mode=::NewEnumerator3
    Mesh Uniform Scale=&Lerp_Float(
     A=&RandomRangeFloat(Minimum=4, Maximum=5)
     B=&RandomRangeFloat(Minimum=0.5, Maximum=1)
     Alpha=&MakeFloatFromInt(Int=$Particles.PartitionParticles.Partition)
    )
    UsePositionOffset=false
    Write Material Random=true
   )
   /SetVariables(
    Particles.Color_Flower:LinearColor=&RandomRangeLinearColor(Minimum=<rgba 1 0.2042 0.8167 1>, Maximum=<rgba 1 0.2232 0.1403 1>)
    Particles.Color_LeafTest:LinearColor=&RandomRangeLinearColor(Minimum=<rgba 0.0958 0.093 0.0036 1>, Maximum=<rgba 0.4697 0.4277 0.0179 1>)
   ) {isSetParameters=true, module=SetVariables_#5}
   &ShapeLocation(MaskSpawnGroup=false, Override Seed=false, UseOwnerScale=true)
   /SetVariables(Particles.MeshIndex=&RandomRangeInt(Minimum=0, Maximum=2)) {isSetParameters=true}
  ]
  particleUpdateScript [
   &ParticleState()
   &UpdateMeshOrientation(
    Facing Direction=&Subtract_Position(A=$Particles.Position, B=$Particles.Previous.Position)
    Facing Interpolation=true
    Facing Interpolation Rate=3
    Orientation Method=::NewEnumerator1
    Override Quat=false
    Stabilization Mode:ENiagara_OrientVectorCount=::NewEnumerator0
    TransformFacing=false
   )
   &UpdateMeshOrientation() {module=UpdateMeshOrientation001}
   &ScaleMeshSize(Scale Factor=&VectorFromFloat(Value=&FloatFromCurve(FloatCurve=DataInterface() {propertyValues {Curve {keys [
          {interpMode=RCIM_Cubic, tangentMode=RCTM_User, time=0, value=1}
          {arriveTangent=-3.7272, interpMode=RCIM_Cubic, leaveTangent=-3.7272, tangentMode=RCTM_User, time=1, value=0}
         ]}}})))
   &DynamicMaterialParameters(
    Index 0 Param 4=&FloatFromCurve(FloatCurve=DataInterface() {propertyValues {Curve {keys [{interpMode=RCIM_Cubic, tangentMode=RCTM_User, time=0, value=-2}, {interpMode=RCIM_Cubic, tangentMode=RCTM_User, time=1, value=9}]}}})
    Module.Param0WriteEnabled=true
    Module.Param1WriteEnabled=true
    Module.Param2WriteEnabled=true
    Module.Param3WriteEnabled=true
   ) {module=DynamicMaterialParameters001}
   &InheritSourceMovement(
    Applied Position Scale=&RandomRangeVector(Minimum=<0.8 0.8 0.8>, Maximum=<1 1 1>)
    ApplyForce=false
    ApplyPosition=true
    ApplyPositionFalloff=false
   )
   &SolveForcesAndVelocity(Clamp Velocity=false, Limit Acceleration=false)
  ]
  rendererClasses [&NiagaraMeshRendererProperties]
  renderers [{rendererClass=&NiagaraMeshRendererProperties, rendererIndex=0}]
 }
 Flower_Petals_DetatchEvent_R {
  emitterUpdateScript [
   &EmitterState(Life Cycle Mode=::NewEnumerator1, Loop Duration=1, Recalculate Duration Each Loop=true, UseLoopDelay=false)
   &SpawnRate(Spawn Probability=&RandomRangeFloat(Minimum=0.5, Maximum=1), SpawnRate=10, Use Spawn Probability=true) {enabled=false}
  ]
  particleSpawnScript [
   &PartitionParticles(Number Of Partitions=3)
   /SetVariables(Particles.VisibilityTag=$Particles.PartitionParticles.Partition) {isSetParameters=true, module=SetVariables_#4}
   &InitializeParticle(
    Color Maximum=<rgba 0.5175 0.3203 0 1>
    Color Minimum=<rgba 0.361 0.3503 0.0138 1>
    Color Mode=::NewEnumerator2
    Lifetime=&RandomRangeFloat(Minimum=0.5, Maximum=1.25)
    Mesh Scale Mode=::NewEnumerator3
    Mesh Uniform Scale=&Lerp_Float(
     A=&RandomRangeFloat(Minimum=4, Maximum=5)
     B=&RandomRangeFloat(Minimum=0.5, Maximum=1)
     Alpha=&MakeFloatFromInt(Int=$Particles.PartitionParticles.Partition)
    )
    UsePositionOffset=false
    Write Material Random=true
   )
   /SetVariables(
    Particles.Color_Flower:LinearColor=&RandomRangeLinearColor(Minimum=<rgba 1 0.2042 0.8167 1>, Maximum=<rgba 1 0.2232 0.1403 1>)
    Particles.Color_LeafTest:LinearColor=&RandomRangeLinearColor(Minimum=<rgba 0.0958 0.093 0.0036 1>, Maximum=<rgba 0.4697 0.4277 0.0179 1>)
   ) {isSetParameters=true, module=SetVariables_#5}
   &ShapeLocation(MaskSpawnGroup=false, Override Seed=false, UseOwnerScale=true)
   /SetVariables(Particles.MeshIndex=&RandomRangeInt(Minimum=0, Maximum=2)) {isSetParameters=true}
  ]
  particleUpdateScript [
   &ParticleState()
   &UpdateMeshOrientation(
    Facing Direction=&Subtract_Position(A=$Particles.Position, B=$Particles.Previous.Position)
    Facing Interpolation=true
    Facing Interpolation Rate=3
    Orientation Method=::NewEnumerator1
    Override Quat=false
    Stabilization Mode:ENiagara_OrientVectorCount=::NewEnumerator0
    TransformFacing=false
   )
   &UpdateMeshOrientation() {module=UpdateMeshOrientation001}
   &ScaleMeshSize(Scale Factor=&VectorFromFloat(Value=&FloatFromCurve(FloatCurve=DataInterface() {propertyValues {Curve {keys [
          {interpMode=RCIM_Cubic, tangentMode=RCTM_User, time=0, value=1}
          {arriveTangent=-3.7272, interpMode=RCIM_Cubic, leaveTangent=-3.7272, tangentMode=RCTM_User, time=1, value=0}
         ]}}})))
   &DynamicMaterialParameters(
    Index 0 Param 4=&FloatFromCurve(FloatCurve=DataInterface() {propertyValues {Curve {keys [{interpMode=RCIM_Cubic, tangentMode=RCTM_User, time=0, value=-2}, {interpMode=RCIM_Cubic, tangentMode=RCTM_User, time=1, value=9}]}}})
    Module.Param0WriteEnabled=true
    Module.Param1WriteEnabled=true
    Module.Param2WriteEnabled=true
    Module.Param3WriteEnabled=true
   ) {module=DynamicMaterialParameters001}
   &InheritSourceMovement(
    Applied Position Scale=&RandomRangeVector(Minimum=<0.8 0.8 0.6>, Maximum=<0.7 1 0.8>)
    ApplyForce=false
    ApplyPosition=true
    ApplyPositionFalloff=false
   )
   &SolveForcesAndVelocity(Clamp Velocity=false, Limit Acceleration=false)
  ]
  rendererClasses [&NiagaraMeshRendererProperties]
  renderers [{rendererClass=&NiagaraMeshRendererProperties, rendererIndex=0}]
 }
 GlowDots {
  emitterUpdateScript [
   &EmitterState(Life Cycle Mode=::NewEnumerator1, Loop Duration=2, UseLoopDelay=false)
   &SpawnRate(SpawnRate=&RandomRangeFloat(Minimum=9, Maximum=22), Use Spawn Probability=false)
   &SpawnRate(SpawnGroup=1, SpawnRate=&RandomRangeFloat(Minimum=9, Maximum=22), Use Spawn Probability=false) {module=SpawnRate001}
  ]
  particleSpawnScript [
   &InitializeParticle(
    Color Maximum=<rgba 3.1987 7 0.7529 1>
    Color Minimum=<rgba 4.0615 7 0 1>
    Color Mode=::NewEnumerator2
    Lifetime=&RandomRangeFloat(Minimum=0.2, Maximum=0.5)
    Sprite Size Mode=::NewEnumerator1
    Uniform Sprite Size Min=3
    UsePositionOffset=false
    Write Material Random=true
   )
   &ShapeLocation(MaskSpawnGroup=false, Override Seed=false, Sphere Radius=11, UseOwnerScale=true)
   &AddVelocity(
    Cone Axis=<0 0 1>
    Module.UseConeFalloff=false
    Module.UseVelDistribution=false
    Override Seed=false
    Velocity Mode=::NewEnumerator2
    Velocity Speed=&RandomRangeFloat(Minimum=99, Maximum=222)
   )
   &SkeletalMeshLocation(
    Bone / Socket Index=&SelectIntFromArray(Direct Array Index=$Emitter.SpawnGroup, Int Selection Array=DataInterface() {propertyValues {IntData [0,1]}})
    Bone Sampling Mode=::NewEnumerator4
    MaskSpawnGroup=false
    Skeletal Mesh=$User.Player
   )
  ]
  particleUpdateScript [
   &ParticleState()
   &ScaleColor(
    Scale Alpha=&FloatFromCurve(FloatCurve=DataInterface() {propertyValues {Curve {keys [<at -0.0046 1.0113 RCIM_Cubic>,<at 1.0011 0.002 RCIM_Cubic>]}}})
    ScaleA=true
    ScaleRGB=false
   )
   &Drag(
    Drag=&FloatFromCurve(
     FloatCurve=DataInterface() {propertyValues {Curve {keys [{interpMode=RCIM_Cubic, tangentMode=RCTM_User, time=0, value=0}, {interpMode=RCIM_Cubic, tangentMode=RCTM_User, time=1, value=1}]}}}
     Scale Curve=5
    )
    Use Linear Drag=true
    Use Rotational Drag=true
   )
   &SolveForcesAndVelocity(Clamp Velocity=false, Limit Acceleration=false)
  ]
  rendererClasses [&NiagaraSpriteRendererProperties]
  renderers [{rendererClass=&NiagaraSpriteRendererProperties}]
 }
 HandGlow_Decal {
  emitterUpdateScript [
   &EmitterState()
   &SpawnBurst_Instantaneous(Spawn Count=2, Use Spawn Probability=false, UseLoopCountLimit=true)
  ]
  particleSpawnScript [
   &InitializeParticle(
    Color=<rgba 0.4351 0.5045 0.0543 0.5>
    Lifetime=0.7
    Sprite Size Mode=::NewEnumerator3
    Uniform Sprite Size=50
    UsePositionOffset=false
    Write Material Random=true
   )
   &PartitionParticles()
  ]
  particleUpdateScript [
   &ParticleState(
    Kill Particles When Lifetime Has Elapsed=false
    Let Infinitely Lived Particles Die When Emitter Deactivates=true
    Loop Particles Lifetime=true
   )
   &SkeletalMeshLocation(
    Bone / Socket Index=&SelectIntFromArray(Direct Array Index=$Particles.PartitionParticles.Partition, Int Selection Array=DataInterface() {propertyValues {IntData [1,0]}})
    Bone Sampling Mode=::NewEnumerator4
    MaskSpawnGroup=false
    Skeletal Mesh=$User.Player
   )
   &ScaleColor(
    Scale Alpha=&FloatFromCurve(FloatCurve=DataInterface() {propertyValues {Curve {keys [<at 0.0001 0.3382>,<at 0.1393 0.6251>,<at 0.3202 0.5233>,<at 0.4613 0.8934>,<at 0.6241 0.3845>,<at 0.7724 0.6066>,<at 1.0006 0.3567>]}}})
    ScaleA=true
    ScaleRGB=false
   )
  ]
  rendererClasses [&NiagaraSpriteRendererProperties]
  renderers [{rendererClass=&NiagaraSpriteRendererProperties}]
 }
 HandWraps {
  emitterUpdateScript [
   &EmitterState(
    Inactive Response=::NewEnumerator1
    Life Cycle Mode=::NewEnumerator1
    Loop Behavior=::NewEnumerator1
    Loop Delay=0.5502
    Loop Duration Mode=::NewEnumerator1
    UseLoopDelay=true
   )
   &SpawnBurst_Instantaneous(Spawn Count=6, Use Spawn Probability=false, UseLoopCountLimit=false)
   &SpawnBurst_Instantaneous(Spawn Count=6, Spawn Group=1, Use Spawn Probability=false, UseLoopCountLimit=false) {module=SpawnBurst_Instantaneous001}
  ]
  particleSpawnScript [
   &InitializeParticle(
    Color Maximum=<rgba 0.1037 0.1958 0.0621 1>
    Color Minimum=<rgba 0.4916 0.6871 0.1692 1>
    Color Mode=::NewEnumerator2
    Mesh Scale Mode=::NewEnumerator3
    Mesh Uniform Scale=&Lerp_Float(
     A=&RandomRangeFloat(Minimum=0.2, Maximum=0.3)
     B=&RandomRangeFloat(Minimum=0.2, Maximum=0.3)
     Alpha=&Multiply_FloatByInt(Float=1, Integer=$Emitter.SpawnGroup)
    )
    UsePositionOffset=false
    Write Material Random=true
   )
   &InitialMeshOrientation(
    Mesh Axis To Orient=<0 0 1>
    Mesh Orientation Mode=::NewEnumerator2
    Orientation Vector=&RotateVectorByQuaternion(Vector To Rotate=<1 0 0>, Quaternion=$Particles.SkeletalMeshLocation001.SampledRotation)
    Rotation=&Lerp_Vector(
     A=&RandomRangeVector(Minimum=<0.1 1 1>, Maximum=<0.5 1 1>)
     B=&RandomRangeVector(Minimum=<0.1 0.9 1>, Maximum=<0.5 0.9 1>)
     Alpha=&Multiply_FloatByInt(Float=1, Integer=$Emitter.SpawnGroup)
    )
    Rotation Coordinate Space=::NewEnumerator0
    Use Rotation Vector=true
   )
  ]
  particleUpdateScript [
   &ParticleState(Kill Particles When Lifetime Has Elapsed=false, Let Infinitely Lived Particles Die When Emitter Deactivates=true)
   &SkeletalMeshLocation(
    Bone / Socket Index=&SelectIntFromArray(Direct Array Index=&SpawnGroup(), Int Selection Array=DataInterface() {propertyValues {IntData [0,3]}})
    Bone Sampling Mode=::NewEnumerator4
    MaskSpawnGroup=false
    Sampled Position Offset=<1 0 0>
    Skeletal Mesh=$User.Player
   )
   &DynamicMaterialParameters(
    Index 0 Param 1=9
    Index 0 Param 2=9
    Index 0 Param 3=1
    Module.Param0WriteEnabled=true
    Module.Param1WriteEnabled=true
    Module.Param2WriteEnabled=true
    Module.Param3WriteEnabled=true
   ) {module=DynamicMaterialParameters002}
   &CameraOffset(Camera Offset Amount=2) {enabled=false}
   &SolveForcesAndVelocity(Clamp Velocity=false, Limit Acceleration=false)
  ]
  rendererClasses [&NiagaraMeshRendererProperties]
  renderers [{rendererClass=&NiagaraMeshRendererProperties, rendererIndex=0}]
 }
 HandWraps001 {
  emitterUpdateScript [
   &EmitterState(
    Inactive Response=::NewEnumerator1
    Life Cycle Mode=::NewEnumerator1
    Loop Behavior=::NewEnumerator1
    Loop Delay=0.5502
    Loop Duration Mode=::NewEnumerator1
    UseLoopDelay=true
   )
   &SpawnBurst_Instantaneous(Spawn Count=6, Use Spawn Probability=false, UseLoopCountLimit=false)
   &SpawnBurst_Instantaneous(Spawn Count=6, Spawn Group=1, Use Spawn Probability=false, UseLoopCountLimit=false) {module=SpawnBurst_Instantaneous001}
  ]
  particleSpawnScript [
   &InitializeParticle(
    Color Maximum=<rgba 0.1037 0.1958 0.0621 1>
    Color Minimum=<rgba 0.4916 0.6871 0.1692 1>
    Color Mode=::NewEnumerator2
    Mesh Scale Mode=::NewEnumerator1
    Mesh Uniform Scale Max=0.08
    Mesh Uniform Scale Min=0.05
    UsePositionOffset=false
    Write Material Random=true
   )
   &InitialMeshOrientation(
    Mesh Axis To Orient=<0 0 1>
    Mesh Orientation Mode=::NewEnumerator2
    Orientation Vector=&RotateVectorByQuaternion(Vector To Rotate=<1 0 0>, Quaternion=$Particles.SkeletalMeshLocation001.SampledRotation)
    Rotation=&Lerp_Vector(
     A=&RandomRangeVector(Minimum=<0.1 1 1>, Maximum=<0.5 1 1>)
     B=&RandomRangeVector(Minimum=<0.1 0.9 1>, Maximum=<0.5 0.9 1>)
     Alpha=&Multiply_FloatByInt(Float=1, Integer=$Emitter.SpawnGroup)
    )
    Rotation Coordinate Space=::NewEnumerator0
    Use Rotation Vector=true
   )
  ]
  particleUpdateScript [
   &ParticleState(Kill Particles When Lifetime Has Elapsed=false, Let Infinitely Lived Particles Die When Emitter Deactivates=true)
   &SkeletalMeshLocation(
    Bone / Socket Index=&SelectIntFromArray(Direct Array Index=&SpawnGroup(), Int Selection Array=DataInterface() {propertyValues {IntData [0,3]}})
    Bone Sampling Mode=::NewEnumerator4
    MaskSpawnGroup=false
    Sampled Position Offset=<1 0 0>
    Skeletal Mesh=$User.Player
   )
   &DynamicMaterialParameters(
    Index 0 Param 1=9
    Index 0 Param 2=9
    Index 0 Param 3=1
    Module.Param0WriteEnabled=true
    Module.Param1WriteEnabled=true
    Module.Param2WriteEnabled=true
    Module.Param3WriteEnabled=true
   ) {module=DynamicMaterialParameters002}
   &CameraOffset(Camera Offset Amount=2) {enabled=false}
   &SolveForcesAndVelocity(Clamp Velocity=false, Limit Acceleration=false)
  ]
  rendererClasses [&NiagaraMeshRendererProperties]
  renderers [{rendererClass=&NiagaraMeshRendererProperties, rendererIndex=0}]
 }
 LeafBurstSpawn {
  simTarget=GPUComputeSim
  emitterUpdateScript [
   &EmitterState(Life Cycle Mode=::NewEnumerator1, Loop Behavior=::NewEnumerator1, Loop Delay=0.15, Loop Duration=0.5, UseLoopDelay=true)
   &SpawnRate(SpawnRate=99, Use Spawn Probability=false) {module=SpawnRate002}
  ]
  particleSpawnScript [
   &PartitionParticles()
   &InitializeParticle(
    Color=&RandomRangeLinearColor(Minimum=<rgba 0.0958 0.093 0.0036 1>, Maximum=<rgba 0.4697 0.4277 0.0179 1>)
    Lifetime Max=1
    Lifetime Min=0.3
    Lifetime Mode=::NewEnumerator1
    Mesh Scale Max=<0.25 0.1 0.25>
    Mesh Scale Min=<0.5 0.5 0.5>
    Mesh Scale Mode=::NewEnumerator2
    Sprite Size Mode=::NewEnumerator1
    Uniform Sprite Size Min=3
    UsePositionOffset=false
    Write Material Random=true
   )
   &SkeletalMeshLocation(
    Bone / Socket Index=&SelectIntFromArray(Direct Array Index=$Particles.PartitionParticles.Partition, Int Selection Array=DataInterface() {propertyValues {IntData [0,1]}})
    Bone Sampling Mode=::NewEnumerator4
    MaskSpawnGroup=false
    Skeletal Mesh=$User.Player
   )
   &ShapeLocation(MaskSpawnGroup=false, Override Seed=false, Sphere Radius=1, UseOwnerScale=true)
   &AddVelocity(
    Cone Axis=<0 0 1>
    Module.UseConeFalloff=false
    Module.UseVelDistribution=false
    Override Seed=false
    Velocity Mode=::NewEnumerator2
    Velocity Speed=&RandomRangeFloat(Minimum=99, Maximum=222)
   )
   &InitialMeshOrientation(Use Rotation Vector=false)
   &AddVelocity() {module=AddVelocity001}
  ]
  particleUpdateScript [
   &ParticleState()
   &UpdateMeshOrientation()
   &ScaleColor(
    Scale Alpha=&FloatFromCurve(FloatCurve=DataInterface() {propertyValues {Curve {keys [<at -0.0046 1.0113 RCIM_Cubic>,<at 1.0011 0.002 RCIM_Cubic>]}}})
    ScaleA=true
    ScaleRGB=false
   )
   &Drag(
    Drag=&FloatFromCurve(
     FloatCurve=DataInterface() {propertyValues {Curve {keys [{interpMode=RCIM_Cubic, tangentMode=RCTM_User, time=0, value=0}, {interpMode=RCIM_Cubic, tangentMode=RCTM_User, time=1, value=1}]}}}
     Scale Curve=5
    )
    Use Linear Drag=true
    Use Rotational Drag=true
   )
   &CurlNoiseForce(Pan Noise=false)
   &SolveForcesAndVelocity(Clamp Velocity=false, Limit Acceleration=false)
   &ScaleMeshSize(Scale Factor=&VectorFromFloat(Value=&FloatFromCurve(FloatCurve=DataInterface() {propertyValues {Curve {keys [
          {interpMode=RCIM_Cubic, tangentMode=RCTM_User, time=0, value=0}
          <at 0.25 1>
          {interpMode=RCIM_Cubic, tangentMode=RCTM_User, time=0.75, value=1}
          {tangentMode=RCTM_User, time=1, value=0}
         ]}}})))
   &CameraOffset(Camera Offset Amount=5)
  ]
  rendererClasses [&NiagaraMeshRendererProperties]
  renderers [{rendererClass=&NiagaraMeshRendererProperties, rendererIndex=0}]
 }
 Leaf_Spawn {
  emitterUpdateScript [
   &EmitterState(Life Cycle Mode=::NewEnumerator1, Loop Duration=2, UseLoopDelay=false)
   &SpawnRate(
    Spawn Probability=&RandomRangeFloat(Minimum=0.3, Maximum=0.5)
    SpawnRate=&Lerp_Float(A=4, B=15, Alpha=$User.ParaGlideLeanAlpha)
    Use Spawn Probability=true
   )
   &SpawnRate(
    Spawn Probability=&RandomRangeFloat(Minimum=0.3, Maximum=0.5)
    SpawnGroup=1
    SpawnRate=&Lerp_Float(A=4, B=15, Alpha=$User.ParaGlideLeanAlpha)
    Use Spawn Probability=true
   ) {module=SpawnRate002}
  ]
  particleSpawnScript [
   &InitializeParticle(
    Color=&RandomRangeLinearColor(Minimum=<rgba 0.0958 0.093 0.0036 1>, Maximum=<rgba 0.4697 0.4277 0.0179 1>)
    Lifetime Max=0.3
    Lifetime Min=0.1
    Lifetime Mode=::NewEnumerator1
    Mesh Scale Mode=::NewEnumerator3
    Mesh Uniform Scale=2
    UsePositionOffset=false
    Write Material Random=true
   )
   &SkeletalMeshLocation(
    Bone / Socket Index=&SelectIntFromArray(Direct Array Index=$Emitter.SpawnGroup, Int Selection Array=DataInterface() {propertyValues {IntData [0,1]}})
    Bone Sampling Mode=::NewEnumerator4
    MaskSpawnGroup=false
    Skeletal Mesh=$User.Player
   )
   &AddVelocity(
    Cone Axis=<0 0 1>
    Module.UseConeFalloff=false
    Module.UseVelDistribution=false
    Override Seed=false
    Velocity Mode=::NewEnumerator2
    Velocity Speed=&RandomRangeFloat(Minimum=99, Maximum=99)
   )
   &InitialMeshOrientation(Use Rotation Vector=false)
  ]
  particleUpdateScript [
   &ParticleState()
   &ScaleMeshSize(Scale Factor=&VectorFromFloat(Value=&FloatFromCurve(FloatCurve=DataInterface() {propertyValues {Curve {keys [
          {interpMode=RCIM_Cubic, tangentMode=RCTM_User, time=0, value=1}
          {arriveTangent=-4.9305, interpMode=RCIM_Cubic, leaveTangent=-4.9305, tangentMode=RCTM_User, time=1, value=0}
         ]}}})))
   &ScaleColor(
    Scale Alpha=&FloatFromCurve(FloatCurve=DataInterface() {propertyValues {Curve {keys [<at -0.0046 1.0113 RCIM_Cubic>,<at 1.0011 0.002 RCIM_Cubic>]}}})
    ScaleA=true
    ScaleRGB=false
   )
   &AccelerationForce(Acceleration=&Multiply_VectorByFloat(Vector=<0 0 1>, Float=555))
   &Drag(
    Drag=&FloatFromCurve(
     FloatCurve=DataInterface() {propertyValues {Curve {keys [{interpMode=RCIM_Cubic, tangentMode=RCTM_User, time=0, value=0}, {interpMode=RCIM_Cubic, tangentMode=RCTM_User, time=1, value=1}]}}}
     Scale Curve=5
    )
    Use Linear Drag=true
    Use Rotational Drag=true
   )
   &SolveForcesAndVelocity(Clamp Velocity=false, Limit Acceleration=false)
   &ScaleSpriteSizeBySpeed()
   &CameraOffset(Camera Offset Amount=-20)
  ]
  rendererClasses [&NiagaraMeshRendererProperties]
  renderers [{rendererClass=&NiagaraMeshRendererProperties, rendererIndex=0}]
 }
 Leafs_AroundPlayer {
  simTarget=GPUComputeSim
  emitterUpdateScript [
   &EmitterState()
   &SpawnRate(
    SpawnRate=&FN_FrontEndFloatModifier(Gameplay=&Lerp_Float(A=10, B=20, Alpha=$User.ParaGlideLeanAlpha), Front End=20)
    Use Spawn Probability=false
   )
  ]
  particleSpawnScript [
   &InitializeParticle(
    Color=&RandomRangeLinearColor(Minimum=<rgba 0.0958 0.093 0.0036 1>, Maximum=<rgba 0.4697 0.4277 0.0179 1>)
    Lifetime Min=0.5
    Lifetime Mode=::NewEnumerator1
    Mesh Scale Max=<0.25 0.1 0.25>
    Mesh Scale Min=<0.5 0.5 0.5>
    Mesh Scale Mode=::NewEnumerator2
    Sprite Rotation Mode=::NewEnumerator2
    Sprite Size Max=<18 17>
    Sprite Size Min=<5 6>
    Sprite Size Mode=::NewEnumerator2
    UsePositionOffset=false
    Write Material Random=true
   ) {module=InitializeParticle001}
   &ShapeLocation(
    MaskSpawnGroup=false
    Offset=<0 0 -75>
    Offset Mode=::NewEnumerator0
    Sphere Distribution=::NewEnumerator1
    Sphere Radius=&RandomRangeFloat(Minimum=75, Maximum=150)
    U Position=&RandomRangeFloat(Minimum=0, Maximum=1)
    UseOwnerScale=true
    V Position=&RandomRangeFloat(Minimum=0, Maximum=1)
   )
   &InheritVelocity(Inherited Velocity Amount Scale=<0 0 -0.12>, Speed Limit=false, UseSpeedThreshold=false)
   &AddVelocity(Velocity=&FN_FrontEndVector3Modifier(
     Gameplay=&Multiply_Vector(A=&VectorFromFloat(Value=$User.ParaGlideLeanAlpha), B=<0 0 150>)
     Front End=&Lerp_Position(A:NiagaraPosition=<0 0 50>, B:NiagaraPosition=<1 -35 300>, Alpha:NiagaraFloat=$User.ParaGlideLeanAlpha)
    ))
   &AccelerationForce(Acceleration=<0 0 99>)
   &VortexForce(UseOriginOffset=false)
   &ApplyInitialForces()
   &InitialMeshOrientation(Use Rotation Vector=false)
  ]
  particleUpdateScript [
   &ParticleState()
   &ScaleMeshSize(Scale Factor=&VectorFromFloat(Value=&FN_FrontEndFloatModifier(
      Gameplay=&FloatFromCurve(FloatCurve=DataInterface() {propertyValues {Curve {keys [
           {arriveTangent=4.1785, interpMode=RCIM_Cubic, leaveTangent=4.1785, tangentMode=RCTM_User, time=0, value=0}
           {arriveTangent=1.1179, interpMode=RCIM_Cubic, leaveTangent=1.1179, tangentMode=RCTM_User, time=0.2042, value=1}
           {arriveTangent=0.0224, interpMode=RCIM_Cubic, leaveTangent=0.0224, tangentMode=RCTM_User, time=1, value=1}
          ]}}})
      Front End=1
     )))
   &ScaleMeshSize(Scale Factor=&VectorFromFloat(Value=&FloatFromCurve(FloatCurve=DataInterface() {propertyValues {Curve {keys [
          {tangentMode=RCTM_User, time=0, value=1}
          {arriveTangent=-0.0405, interpMode=RCIM_Cubic, leaveTangent=-0.0405, tangentMode=RCTM_User, time=0.7009, value=1}
          {arriveTangent=-8.0089, interpMode=RCIM_Cubic, leaveTangent=-8.0089, tangentMode=RCTM_User, time=1, value=0}
         ]}}}))) {module=ScaleMeshSize001}
   &UpdateMeshOrientation()
   &CurlNoiseForce(Noise Frequency=1000, Noise Strength=200, Pan Noise=false)
   &SolveForcesAndVelocity(Clamp Velocity=false, Limit Acceleration=false)
   &ScaleMeshSizeBySpeed(Max Scale Factor=<1.2 1.2 1.2>, Min Scale Factor:Vector3f=<0.5 0.5 0.5>)
  ]
  rendererClasses [&NiagaraMeshRendererProperties]
  renderers [{rendererClass=&NiagaraMeshRendererProperties, rendererIndex=0}]
 }
 SpeedLines_Spawn {
  simTarget=GPUComputeSim
  emitterUpdateScript [
   &EmitterState(Life Cycle Mode=::NewEnumerator1, Loop Behavior=::NewEnumerator1, Loop Duration Mode=::NewEnumerator1, UseLoopDelay=false)
   &SpawnRate(SpawnRate=&Lerp_Float(A=0, B=222, Alpha=$User.ParaGlideLeanAlpha), Use Spawn Probability=false)
  ]
  particleSpawnScript [
   &InitializeParticle(
    Lifetime=0.1
    Sprite Size Max=<3 10>
    Sprite Size Min=<2 2>
    Sprite Size Mode=::NewEnumerator2
    UsePositionOffset=false
    Write Material Random=true
   )
   &ShapeLocation(
    MaskSpawnGroup=false
    Override Seed=false
    Rotation Angle=180
    Rotation Angle Type:ENiagara_AngleInput=::NewEnumerator0
    Rotation Axis=$System.SkydiveVector
    Rotation Mode=::NewEnumerator1
    Shape Primitive=::NewEnumerator3
    UseOwnerScale=true
   )
   &AddVelocity(
    Velocity=&TransformVector(Vector=$System.SkydiveVector, Source Space=::World, Destination Space=::Local)
    Velocity Speed Scale=500
   )
   &ApplyInitialForces()
  ]
  particleUpdateScript [
   &ParticleState()
   &SolveForcesAndVelocity(Clamp Velocity=false, Limit Acceleration=false)
   &ScaleSpriteSize(Scale Factor=&Vector2DFromCurve(
     Vector2Curve=DataInterface() {propertyValues {XCurve {keys [<at 0 0.5 RCIM_Cubic>,<at 0.5 1 RCIM_Cubic>]}, YCurve {keys [<at 0 0 RCIM_Cubic>,<at 0.5 1 RCIM_Cubic>]}}}
     ScaleCurve=&Vector2DFromFloat(Value=1)
    ))
   &ScaleSpriteSizeBySpeed(Max Scale Factor=<1 50>, Min Scale Factor=<0.01 0.05>, Velocity Threshold=1)
   &ScaleColor(
    Scale Alpha=&FloatFromCurve(
     FloatCurve=DataInterface() {propertyValues {Curve {keys [
         <at 0 0 RCIM_Cubic>
         <at 0.3759 1>
         {arriveTangent=-2.221, interpMode=RCIM_Cubic, leaveTangent=-2.221, tangentMode=RCTM_User, time=1, value=0}
        ]}}}
     Scale Curve=0.02
    )
    Scale RGB=&VectorFromFloat(Value=2)
    ScaleA=true
    ScaleRGB=true
   )
  ]
  rendererClasses [&NiagaraSpriteRendererProperties]
  renderers [{rendererClass=&NiagaraSpriteRendererProperties}]
 }
 Vine_L {
  emitterUpdateScript [
   &EmitterState(
    Enable Distance Culling=true
    Life Cycle Mode=::NewEnumerator1
    Loop Behavior=::NewEnumerator1
    Loop Duration Mode=::NewEnumerator1
    Max Distance Response=::SleepAndClearParticles
    Module.bUseMaxDistance=true
    Module.bUseMinDistance=true
    Reset Age On Awaken=true
    Scalability Mode=::NewEnumerator1
    UseLoopDelay=false
   )
   &SpawnBurst_Instantaneous(Spawn Count=5, Use Spawn Probability=false, UseLoopCountLimit=false)
  ]
  particleSpawnScript [
   &InitializeParticle(Color=<rgba 0.0071 0.0654 0 1>, Ribbon Width=50, Ribbon Width Mode=::NewEnumerator1, UsePositionOffset=false, Write Material Random=true)
   /SetVariables(Particles.RibbonLinkOrder:NiagaraFloat=&MakeFloatFromInt(Int=&ReturnParticleIDAttributes())) {isSetParameters=true, module=SetVariables_#8}
  ]
  particleUpdateScript [
   &ParticleState(Kill Particles When Lifetime Has Elapsed=false, Loop Particles Lifetime=true)
   &SampleSkeletalMesh(
    AdditionalYawPitchRoll=false
    Bone Sampling Mode:ENiagaraBoneSamplingMode=::NewEnumerator4
    Skeletal Mesh:NiagaraDataInterfaceSkeletalMesh=$User.Player
   )
   /SetVariables(Particles.Position=&AddVectorToPosition(
     Position=$Particles.SampleSkeletalMesh.SampledPosition
     Vector=&Multiply_VectorByFloat(
      Vector=$System.SkydiveVector
      Float=&FloatFromCurve(
       FloatCurve=DataInterface() {propertyValues {Curve {keys [{tangentMode=RCTM_User, time=0, value=0}, {tangentMode=RCTM_User, time=1, value=1}]}}}
       CurveIndex=&ModuloFloat(
        A=&Add_Float(A=&ReturnNormalizedExecIndex(Normalized Index Includes One=false), B=0)
        B=1
       )
       Scale Curve=&FloatFromCurve(
        FloatCurve=DataInterface() {propertyValues {Curve {keys [{interpMode=RCIM_Cubic, tangentMode=RCTM_User, time=0, value=0}, {interpMode=RCIM_Cubic, tangentMode=RCTM_User, time=1, value=1}]}}}
        CurveIndex=$Engine.System.Age
        Scale Curve=&RandomRangeFloat(Minimum=175, Maximum=190)
       )
      )
     )
    )) {isSetParameters=true, module=SetVariables_#9}
   /SetVariables(Particles.Position=&AddVectorToPosition(
     Position=$Particles.Position
     Vector=&Multiply_VectorByFloat(
      Vector=<1 0 0>
      Float=&Waveform(
       Global Input Position=&ReturnNormalizedExecIndex()
       Global Amplitude Scale=&ReturnNormalizedExecIndex()
       "Waveform [1]"=::NewEnumerator0
       "[1] Phase Scale"=-2
       "[1] Frequency"=1
       "[1] Amplitude Min/Max"=<-1 1>
       "[1] Scale"=&Lerp_Float(A=10, B=25, Alpha=$User.ParaGlideLeanAlpha)
      )
     )
    )) {isSetParameters=true, module=SetVariables_#10}
   /SetVariables(Particles.Position=&AddVectorToPosition(
     Position=$Particles.Position
     Vector=&Multiply_VectorByFloat(
      Vector=<0 1 0>
      Float=&Waveform(
       Global Input Position=&ReturnNormalizedExecIndex()
       Global Amplitude Scale=&ReturnNormalizedExecIndex()
       "Waveform [1]"=::NewEnumerator1
       "[1] Phase Scale"=-2
       "[1] Frequency"=1
       "[1] Amplitude Min/Max"=<-1 1>
       "[1] Scale"=&Lerp_Float(A=10, B=20, Alpha=$User.ParaGlideLeanAlpha)
      )
     )
    )) {isSetParameters=true, module=SetVariables_#11}
   &ScaleRibbonWidth(Ribbon Width Scale=&FloatFromCurve(
     FloatCurve=DataInterface() {propertyValues {Curve {keys [{interpMode=RCIM_Cubic, tangentMode=RCTM_User, time=0, value=0.1}, {interpMode=RCIM_Cubic, tangentMode=RCTM_User, time=1, value=0.7}]}}}
     CurveIndex=$Particles.RibbonLinkOrder
    ))
  ]
  rendererClasses [&NiagaraRibbonRendererProperties]
  renderers [{rendererClass=&NiagaraRibbonRendererProperties}]
 }
 Vine_R {
  emitterUpdateScript [
   &EmitterState(
    Enable Distance Culling=true
    Life Cycle Mode=::NewEnumerator1
    Loop Behavior=::NewEnumerator1
    Loop Duration Mode=::NewEnumerator1
    Max Distance Response=::SleepAndClearParticles
    Module.bUseMaxDistance=true
    Module.bUseMinDistance=true
    Reset Age On Awaken=true
    Scalability Mode=::NewEnumerator1
    UseLoopDelay=false
   )
   &SpawnBurst_Instantaneous(Spawn Count=5, Use Spawn Probability=false, UseLoopCountLimit=false)
  ]
  particleSpawnScript [
   &InitializeParticle(
    Color=<rgba 0.0071 0.0654 0 1>
    Ribbon Width=50
    Ribbon Width Mode=::NewEnumerator1
    Sprite Size Mode=::NewEnumerator3
    UsePositionOffset=false
    Write Material Random=true
   )
   /SetVariables(Particles.RibbonLinkOrder:NiagaraFloat=&MakeFloatFromInt(Int=&ReturnParticleIDAttributes())) {isSetParameters=true, module=SetVariables_#8}
  ]
  particleUpdateScript [
   &ParticleState(Kill Particles When Lifetime Has Elapsed=false, Loop Particles Lifetime=true)
   &SampleSkeletalMesh(
    AdditionalYawPitchRoll=false
    Bone / Socket Index=1
    Bone Sampling Mode:ENiagaraBoneSamplingMode=::NewEnumerator4
    Skeletal Mesh:NiagaraDataInterfaceSkeletalMesh=$User.Player
   )
   /SetVariables(Particles.Position=&AddVectorToPosition(
     Position=$Particles.SampleSkeletalMesh.SampledPosition
     Vector=&Multiply_VectorByFloat(
      Vector=$System.SkydiveVector
      Float=&FloatFromCurve(
       FloatCurve=DataInterface() {propertyValues {Curve {keys [{tangentMode=RCTM_User, time=0, value=0}, {tangentMode=RCTM_User, time=1, value=1}]}}}
       CurveIndex=&ModuloFloat(
        A=&Add_Float(A=&ReturnNormalizedExecIndex(Normalized Index Includes One=false), B=0)
        B=1
       )
       Scale Curve=&FloatFromCurve(
        FloatCurve=DataInterface() {propertyValues {Curve {keys [{interpMode=RCIM_Cubic, tangentMode=RCTM_User, time=0, value=0}, {interpMode=RCIM_Cubic, tangentMode=RCTM_User, time=1, value=1}]}}}
        CurveIndex=$Engine.System.Age
        Scale Curve=&Lerp_Float(A=&RandomRangeFloat(Minimum=175, Maximum=185), B=210, Alpha=$User.ParaGlideLeanAlpha)
       )
      )
     )
    )) {isSetParameters=true, module=SetVariables_#9}
   /SetVariables(Particles.Position=&AddVectorToPosition(
     Position=$Particles.Position
     Vector=&Multiply_VectorByFloat(
      Vector=<1 0 0>
      Float=&Waveform(
       Global Input Position=&ReturnNormalizedExecIndex()
       Global Amplitude Scale=&ReturnNormalizedExecIndex()
       "Waveform [1]"=::NewEnumerator0
       "[1] Phase Scale"=-2
       "[1] Frequency"=1
       "[1] Amplitude Min/Max"=<-1 1>
       "[1] Scale"=&Lerp_Float(A=10, B=25, Alpha=$User.ParaGlideLeanAlpha)
      )
     )
    )) {isSetParameters=true, module=SetVariables_#10}
   /SetVariables(Particles.Position=&AddVectorToPosition(
     Position=$Particles.Position
     Vector=&Multiply_VectorByFloat(
      Vector=<0 1 0>
      Float=&Waveform(
       Global Input Position=&ReturnNormalizedExecIndex()
       Global Amplitude Scale=&ReturnNormalizedExecIndex()
       "Waveform [1]"=::NewEnumerator1
       "[1] Phase Scale"=-2
       "[1] Frequency"=1
       "[1] Amplitude Min/Max"=<-1 1>
       "[1] Scale"=&Lerp_Float(A=10, B=20, Alpha=$User.ParaGlideLeanAlpha)
      )
     )
    )) {isSetParameters=true, module=SetVariables_#11}
   &ScaleRibbonWidth(Ribbon Width Scale=&FloatFromCurve(
     FloatCurve=DataInterface() {propertyValues {Curve {keys [{interpMode=RCIM_Cubic, tangentMode=RCTM_User, time=0, value=0.1}, {interpMode=RCIM_Cubic, tangentMode=RCTM_User, time=1, value=0.7}]}}}
     CurveIndex=$Particles.RibbonLinkOrder
    ))
  ]
  rendererClasses [&NiagaraRibbonRendererProperties]
  renderers [{rendererClass=&NiagaraRibbonRendererProperties}]
 }
}
summary {
 systemName=NS_Hibiscus_Vines_Trail
 emitters [
  {emitterName=Vine_L, rendererClasses [&NiagaraRibbonRendererProperties]}
  {emitterName=Vine_R, rendererClasses [&NiagaraRibbonRendererProperties]}
  {emitterName=HandGlow_Decal, rendererClasses [&NiagaraSpriteRendererProperties]}
  {emitterName=GlowDots, rendererClasses [&NiagaraSpriteRendererProperties]}
  {emitterName=Flower_Petals_DetatchEvent_R, rendererClasses [&NiagaraMeshRendererProperties]}
  {emitterName=Flower_Petals_DetatchEvent_L, rendererClasses [&NiagaraMeshRendererProperties]}
  {emitterName=Leaf_Spawn, rendererClasses [&NiagaraMeshRendererProperties]}
  {emitterName=LeafBurstSpawn, simTarget=GPUComputeSim, rendererClasses [&NiagaraMeshRendererProperties]}
  {emitterName=Leafs_AroundPlayer, simTarget=GPUComputeSim, rendererClasses [&NiagaraMeshRendererProperties]}
  {emitterName=FlowerPetals_AroundPlayer, simTarget=GPUComputeSim, rendererClasses [&NiagaraMeshRendererProperties]}
  {emitterName=SpeedLines_Spawn, simTarget=GPUComputeSim, rendererClasses [&NiagaraSpriteRendererProperties]}
  {emitterName=Flower_Leafs_L, rendererClasses [&NiagaraMeshRendererProperties]}
  {emitterName=Flower_Leafs_R, rendererClasses [&NiagaraMeshRendererProperties]}
  {emitterName=HandWraps, rendererClasses [&NiagaraMeshRendererProperties]}
  {emitterName=HandWraps001, rendererClasses [&NiagaraMeshRendererProperties]}
 ]
 userVariables [
  User.bIsFrontEnd {type=NiagaraBool}
  User.bIsFrontEndPreview {type=NiagaraBool}
  User.ParaGlideLeanAlpha {type=NiagaraFloat, defaultValue=0, description=""}
  User.Player {
   type=NiagaraDataInterfaceSkeletalMesh
   defaultValue {
    struct=NiagaraExt_VariableValue_DataInterface
    value {dataInterface=&self:NiagaraDataInterfaceSkeletalMesh_0}
   }
  }
 ]
}
